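# Animation Blueprint for the female medium TrafficHat Boba head's hat variant, on the Fortnite base head skeleton.
# It layers jaw, lip, cheek and nose bones copied from BodyMesh onto the input pose.
# While crouching, an additive head rotation partly counters head-versus-neck_02 rotation, weighted per axis.
# On customization load it caches part meshes and sets hat and mask flags, unused here, from the face accessory's asset name and material parameter.
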AnimBlueprint /BRCosmetics/Characters/Player/Female/Medium/Heads/F_MED_TrafficHat_Boba_Head/Meshes/F_MED_TrafficHat_Boba_Head_AnimBP_Hat
target_skeleton=/Game/Characters/Player/Common/Fortnite_Base_Head/Fortnite_Base_Head_Export_Skeleton
graphs [
 AnimGraph {nodes [
   n1 = Layered blend per bone(
    BasePose=Input Pose()
    BlendPoses_0=Copy Pose From Mesh(SourceMeshComponent=Get BodyMesh() -> (BodyMesh:Skeletal Mesh Component Budgeted Object Reference))
   ) {layerSetup [{branchFilters [
       {boneName=C_jaw}
       {boneName=C_lip_lower_mid}
       {boneName=L_lip_lower_outer}
       {boneName=R_lip_lower_outer}
       {boneName=C_lip_upper_mid}
       {boneName=C_nose_bridge}
       {boneName=L_cheek_inner}
       {boneName=L_lip_corner}
       {boneName=L_lip_upper_outer}
       {boneName=R_cheek_inner}
       {boneName=R_lip_corner}
       {boneName=R_lip_upper_outer}
      ]}]}
   n2 = "Transform (Modify) Bone - Bone: head"!(
    ComponentPose=Local To Component(LocalPose=@n1)
    bAlphaBoolEnabled:bool=True
    Translation="0,0,0"
    Rotation=Get head rot() -> (head rot:Rotator)
    Scale="1,1,1"
   ) -> (Pose) {
    alphaInputType=Bool
    rotationMode=BMM_Additive
    rotationSpace=BCS_BoneSpace
    alphaBoolBlend {blendInTime=0.1, blendOption=Linear, blendOutTime=0.2}
    boneToModify {boneName=head}
    rotation=<rot 50 50 50>
   }
   Output Pose(Result=Component To Local(ComponentPose=@n2))
  ]}
 EventGraph {nodes [
   Event BlueprintUpdateAnimation()
   n1 = BreakRotator(InRot=GetDeltaTransformFromRefPose!(self=GetOwnerSkeletalMeshComponent(), BoneName=head, BaseName=neck_02) -> (ReturnValue_Location:Vector, ReturnValue_Rotation:Rotator, ReturnValue_Scale:Vector).ReturnValue_Rotation)
   n2 = SelectFloat(A=Get Y percentage() -> (Y percentage:float), B=0, bPickA=Get bIsCrouching())
   n3 = SelectFloat(A=Get X percentage() -> (X percentage:float), B=0, bPickA=Get bIsCrouching())
   n4 = SelectFloat(A=Get Z percentage() -> (Z percentage:float), B=0, bPickA=Get bIsCrouching())
   > Set head rot(head rot:Rotator=MakeRotator(
     Roll=float * float!(A=@n1, B=-1, C:double=@n3) -> (ReturnValue)
     Pitch=float * float(A=@n1.Pitch, B=-1, C=@n2)
     Yaw=float * float(A=@n1.Yaw, B=-1, C=@n4)
    )) -> (Output_Get:Rotator)
   n5 = Event OnCharacterCustomizationLoaded()
   n6 = Get Array Item with EFort Custom Part Type(Array:Array of Skinned Mesh Component Object References=@n5.CustomMeshParts, PartType=Face) -> (Output:Skinned Mesh Component Object Reference)
   > Is Valid(InputObject=@n6)
   > Set BodyMesh(BodyMesh:Skeletal Mesh Component Budgeted Object Reference=Get Array Item with EFort Custom Part Type(Array:Array of Skeletal Mesh Component Budgeted Object References=@n5.CustomMeshParts) -> (Output:Skeletal Mesh Component Budgeted Object Reference)) -> (Output_Get:Skeletal Mesh Component Budgeted Object Reference)
   > Set FaceAccMesh(FaceAccMesh:Skinned Mesh Component Object Reference=@n6) -> (Output_Get:Skinned Mesh Component Object Reference)
   n7 = GetSkinnedAsset!(self:Skinned Mesh Component Object Reference=Get FaceAccMesh() -> (FaceAccMesh:Skinned Mesh Component Object Reference)) -> (ReturnValue:Skinned Asset Object Reference)
   > Set bIsHatOn(bIsHatOn:bool=Contains(SearchIn=GetDisplayName(Object=@n7), Substring=Get HatText() -> (HatText:String), bSearchFromEnd=true)) -> (Output_Get:bool)
   > n8 = Get Scalar Parameter Value!(
    execute
    self:Material Instance Dynamic Object Reference=GetDynamicMaterialInstance!(MeshComponent:Mesh Component Object Reference=Get FaceAccMesh(), MaterialIndex:int=0) -> (Material Instance Dynamic:Material Instance Dynamic Object Reference)
    ParameterName:Name=Hide Element 01
   ) -> (then, ReturnValue:float) {is=title K2Node_CallFunction}
   > Set bIsMaskOn(bIsMaskOn:bool=NOT Boolean(A="To Bool (Double)"!(Value:double=@n8) -> (ReturnValue:bool))) -> (Output_Get:bool)
  ]}
]
variables [
 bIsHatOn {type=boolean}
 bIsMaskOn {type=boolean}
 BodyMesh {type=SkeletalMeshComponentBudgeted}
 FaceAccMesh {type=SkinnedMeshComponent, default=null}
 HatText {type=string, default=BobaHat}
 head rot {type=Rotator}
 X percentage {type=number, default=0.2}
 Y percentage {type=number, default=0.5}
 Z percentage {type=number, default=0.4}
]
